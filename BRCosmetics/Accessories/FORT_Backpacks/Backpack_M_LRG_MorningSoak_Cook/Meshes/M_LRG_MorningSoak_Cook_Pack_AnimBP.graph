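# Anim Blueprint for the MorningSoak Cook backpack, adding springy secondary motion to its spine, head and limb bones.
# Each update maps owner-mesh spine yaw to additive pack offsets and flags falling without skydiving.
# Pack and pelvis offsets are cached as Presim, feeding two separately tuned Anim Dynamics chains.
# A bool blend selects between the chains, one of which also adds a half-weight linear spring on the pelvis.
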AnimBlueprint /BRCosmetics/Accessories/FORT_Backpacks/Backpack_M_LRG_MorningSoak_Cook/Meshes/M_LRG_MorningSoak_Cook_Pack_AnimBP
target_skeleton=&/M_LRG_MorningSoak_Cook_Pack_Skeleton
graphs [
 AnimGraph {nodes [
   n1 = Anim Dynamics - Bone: dyn_BThing_r(ComponentPose=Local To Component(LocalPose=Use cached pose 'Presim'()), Alpha=1, ExternalForce="0,0,0") {
    angularDampingOverride=1
    angularSpringConstant=7000
    bAngularSpring=true
    bOverrideAngularDamping=true
    bUseGravityOverride=true
    simulationSpace=BoneRelative
    boundBone {boneName=dyn_BThing_r}
    componentLinearAccScale {z=1.5}
    componentLinearVelScale {z=0.4}
    gravityOverride {z=-980}
    physicsBodyDefinitions [{
      boundBone {boneName=dyn_BThing_r}
      boxExtents=<20 20 20>
      constraintSetup {angularLimitsMax=<0 2 0>, angularLimitsMin=<0 -2 -35>, angularTarget=<0 0 0>}
      localJointOffset=<0.25 0 0>
     }]
    relativeSpaceBone {boneName=root}
    simSpaceSettings {maxAngularAcceleration=38.835, maxAngularVelocity=19.3792, simSpaceAngularAlpha=1}
   }
   n2 = Anim Dynamics - Bone: dyn_BCalf_r(ComponentPose=@n1, Alpha=1, ExternalForce="0,0,0") {
    angularDampingOverride=1
    angularSpringConstant=3500
    bAngularSpring=true
    bOverrideAngularDamping=true
    bUseGravityOverride=true
    simulationSpace=RootRelative
    boundBone {boneName=dyn_BCalf_r}
    componentLinearAccScale {z=6.1054}
    componentLinearVelScale {z=4.9901}
    gravityOverride {z=-980}
    physicsBodyDefinitions [{
      boundBone {boneName=dyn_BCalf_r}
      boxExtents=<20 20 20>
      constraintSetup {angularLimitsMax=<0 10 15>, angularLimitsMin=<0 -10 -35>, angularTarget=<1 0 0>}
      localJointOffset=<0.28 0 0>
     }]
    relativeSpaceBone {boneName=root}
    simSpaceSettings {maxAngularAcceleration=50, maxAngularVelocity=10, simSpaceAngularAlpha=1}
   }
   n3 = Anim Dynamics - Bone: dyn_BThing_l(ComponentPose=@n2, Alpha=1, ExternalForce="0,0,0") {
    angularDampingOverride=1
    angularSpringConstant=5000
    bAngularSpring=true
    bOverrideAngularDamping=true
    bUseGravityOverride=true
    simulationSpace=RootRelative
    boundBone {boneName=dyn_BThing_l}
    componentLinearAccScale {z=1.5}
    componentLinearVelScale {z=1}
    gravityOverride {z=-980}
    physicsBodyDefinitions [{
      boundBone {boneName=dyn_BThing_l}
      boxExtents=<30 30 30>
      constraintSetup {angularLimitsMax=<0 2 0>, angularLimitsMin=<0 -2 -35>, angularTarget=<0 0 0>}
      localJointOffset=<0.35 0 0>
     }]
    relativeSpaceBone {boneName=root}
    simSpaceSettings {maxAngularAcceleration=10, maxAngularVelocity=3, simSpaceAngularAlpha=1}
   }
   n4 = Anim Dynamics - Bone: dyn_BCalf_l(ComponentPose=@n3, Alpha=1, ExternalForce="0,0,0") {
    angularDampingOverride=1
    angularSpringConstant=3500
    bAngularSpring=true
    bOverrideAngularDamping=true
    bUseGravityOverride=true
    simulationSpace=RootRelative
    boundBone {boneName=dyn_BCalf_l}
    componentLinearAccScale {z=2}
    componentLinearVelScale {z=3.5}
    gravityOverride {z=-980}
    physicsBodyDefinitions [{
      boundBone {boneName=dyn_BCalf_l}
      boxExtents=<17 17 17>
      constraintSetup {angularLimitsMax=<0 10 15>, angularLimitsMin=<0 -10 -35>, angularTarget=<1 0 0>}
      localJointOffset=<0.2 0 0>
     }]
    relativeSpaceBone {boneName=root}
    simSpaceSettings {maxAngularAcceleration=30, maxAngularVelocity=13, simSpaceAngularAlpha=1}
   }
   n5 = Anim Dynamics - Bone: dyn_BSpine1(ComponentPose=@n4, Alpha=1, ExternalForce="0,0,0") {
    angularDampingOverride=1
    angularSpringConstant=2500
    bAngularSpring=true
    bOverrideAngularDamping=true
    bUseGravityOverride=true
    simulationSpace=RootRelative
    boundBone {boneName=dyn_BSpine1}
    componentAppliedLinearAccClamp {z=4000}
    componentLinearAccScale {z=1.5}
    componentLinearVelScale {z=1}
    gravityOverride {z=-980}
    physicsBodyDefinitions [{
      boundBone {boneName=dyn_BSpine1}
      boxExtents=<15 15 15>
      constraintSetup {angularLimitsMax=<0 5 0>, angularLimitsMin=<0 -5 -15>, angularTarget=<1 0 0>}
      localJointOffset=<1 0 0>
     }]
    relativeSpaceBone {boneName=root}
    simSpaceSettings {maxAngularAcceleration=10, maxAngularVelocity=3, simSpaceAngularAlpha=1}
   }
   n6 = Anim Dynamics - Bone: dyn_BSpine2(ComponentPose=@n5, Alpha=1, ExternalForce="0,0,0") {
    angularDampingOverride=1
    angularSpringConstant=5000
    bAngularSpring=true
    bOverrideAngularDamping=true
    bUseGravityOverride=true
    simulationSpace=RootRelative
    boundBone {boneName=dyn_BSpine2}
    componentLinearAccScale {z=1.5}
    componentLinearVelScale {z=1}
    gravityOverride {z=-980}
    physicsBodyDefinitions [{
      boundBone {boneName=dyn_BSpine2}
      boxExtents=<15 15 15>
      constraintSetup {angularLimitsMax=<0 2 0>, angularLimitsMin=<0 -2 -15>, angularTarget=<1 0 0>}
      localJointOffset=<1 0 0>
     }]
    relativeSpaceBone {boneName=root}
    simSpaceSettings {maxAngularAcceleration=10, maxAngularVelocity=3, simSpaceAngularAlpha=1}
   }
   n7 = Anim Dynamics - Bone: dyn_BHead(ComponentPose=@n6, Alpha=1, ExternalForce="0,0,0") {
    angularDampingOverride=1
    angularSpringConstant=4000
    bAngularSpring=true
    bOverrideAngularDamping=true
    bUseGravityOverride=true
    simulationSpace=RootRelative
    boundBone {boneName=dyn_BHead}
    componentLinearAccScale {z=1}
    componentLinearVelScale {z=0.5}
    gravityOverride {z=-980}
    physicsBodyDefinitions [{
      boundBone {boneName=dyn_BHead}
      boxExtents=<15 15 15>
      constraintSetup {angularLimitsMax=<0 25 1>, angularLimitsMin=<0 -5 -25>, angularTarget=<1 0 0>}
      localJointOffset=<0.5 0 0>
     }]
    relativeSpaceBone {boneName=root}
    simSpaceSettings {maxAngularAcceleration=10, maxAngularVelocity=3, simSpaceAngularAlpha=1}
   }
   n8 = Anim Dynamics - Bone: dyn_BUpperArm_r(ComponentPose=@n7, Alpha=1, ExternalForce="0,0,0") {
    angularDampingOverride=1
    angularSpringConstant=4000
    bAngularSpring=true
    bOverrideAngularDamping=true
    bUseGravityOverride=true
    boundBone {boneName=dyn_BUpperArm_r}
    componentAppliedLinearAccClamp {z=70000}
    componentLinearAccScale {z=4.4055}
    componentLinearVelScale {z=4.7811}
    gravityOverride {z=-980}
    physicsBodyDefinitions [{
      boundBone {boneName=dyn_BUpperArm_r}
      boxExtents=<25 25 25>
      constraintSetup {angularLimitsMax=<0 7.9268 0>, angularLimitsMin=<0 -41.8146 -52.7951>, angularTarget=<1 0 0>}
      localJointOffset=<0.4 0 0>
     }]
    relativeSpaceBone {boneName=root}
    simSpaceSettings {maxAngularAcceleration=30, maxAngularVelocity=7.6567, simSpaceAngularAlpha=1}
   }
   n9 = Anim Dynamics - Bone: dyn_BLoverArm_r(ComponentPose=@n8, Alpha=1, ExternalForce="0,0,0") {
    angularDampingOverride=1
    angularSpringConstant=4000
    bAngularSpring=true
    bOverrideAngularDamping=true
    bUseGravityOverride=true
    simulationSpace=RootRelative
    boundBone {boneName=dyn_BLoverArm_r}
    componentLinearAccScale {z=1.5}
    componentLinearVelScale {z=1}
    gravityOverride {z=-980}
    physicsBodyDefinitions [{
      boundBone {boneName=dyn_BLoverArm_r}
      boxExtents=<20 20 20>
      constraintSetup {angularLimitsMax=<0 5 0>, angularLimitsMin=<0 -12.8 -10>, angularTarget=<0 0 0>}
      localJointOffset=<0.4 0 0>
     }]
    relativeSpaceBone {boneName=root}
    simSpaceSettings {maxAngularAcceleration=10, maxAngularVelocity=3, simSpaceAngularAlpha=1}
   }
   n10 = Anim Dynamics - Bone: dyn_BHand_r(ComponentPose=@n9, Alpha=1, ExternalForce="0,0,0") {
    angularDampingOverride=1
    angularSpringConstant=4000
    bAngularSpring=true
    bOverrideAngularDamping=true
    bUseGravityOverride=true
    simulationSpace=RootRelative
    boundBone {boneName=dyn_BHand_r}
    componentLinearAccScale {z=1.5}
    componentLinearVelScale {z=1}
    gravityOverride {z=-980}
    physicsBodyDefinitions [{
      boundBone {boneName=dyn_BHand_r}
      boxExtents=<20 20 20>
      constraintSetup {angularLimitsMax=<0 5 0>, angularLimitsMin=<0 -20 -5>, angularTarget=<0 0 0>}
      localJointOffset=<0.4 0 0>
     }]
    relativeSpaceBone {boneName=root}
    simSpaceSettings {maxAngularAcceleration=10, maxAngularVelocity=3, simSpaceAngularAlpha=1}
   }
   n11 = Anim Dynamics - Bone: dyn_BUpperArm_l(ComponentPose=@n10, Alpha=1, ExternalForce="0,0,0") {
    angularDampingOverride=1
    angularSpringConstant=5000
    bAngularSpring=true
    bOverrideAngularDamping=true
    bUseGravityOverride=true
    simulationSpace=RootRelative
    boundBone {boneName=dyn_BUpperArm_l}
    componentLinearAccScale {z=2.5}
    componentLinearVelScale {z=1.7}
    gravityOverride {z=-980}
    physicsBodyDefinitions [{
      boundBone {boneName=dyn_BUpperArm_l}
      boxExtents=<25 25 25>
      constraintSetup {angularLimitsMax=<0 25 0>, angularLimitsMin=<0 -5 -27.8>, angularTarget=<0 0 0>}
      localJointOffset=<0.25 0 0>
     }]
    relativeSpaceBone {boneName=root}
    simSpaceSettings {maxAngularAcceleration=30, maxAngularVelocity=15, simSpaceAngularAlpha=1}
   }
   n12 = Anim Dynamics - Bone: dyn_BLoverArm_l(ComponentPose=@n11, Alpha=1, ExternalForce="0,0,0") {
    angularDampingOverride=1
    angularSpringConstant=5000
    bAngularSpring=true
    bOverrideAngularDamping=true
    bUseGravityOverride=true
    simulationSpace=RootRelative
    boundBone {boneName=dyn_BLoverArm_l}
    componentLinearAccScale {z=1.5}
    componentLinearVelScale {z=1}
    gravityOverride {z=-980}
    physicsBodyDefinitions [{
      boundBone {boneName=dyn_BLoverArm_l}
      boxExtents=<25 25 25>
      constraintSetup {angularLimitsMax=<0 15 0>, angularLimitsMin=<0 -5 -10>, angularTarget=<0 0 0>}
      localJointOffset=<0.35 0 0>
     }]
    relativeSpaceBone {boneName=root}
    simSpaceSettings {maxAngularAcceleration=10, maxAngularVelocity=3, simSpaceAngularAlpha=1}
   }
   n13 = Anim Dynamics - Bone: dyn_BHand_l(ComponentPose=@n12, Alpha=1, ExternalForce="0,0,0") {
    angularDampingOverride=1
    angularSpringConstant=5000
    bAngularSpring=true
    bOverrideAngularDamping=true
    bUseGravityOverride=true
    simulationSpace=RootRelative
    boundBone {boneName=dyn_BHand_l}
    componentLinearAccScale {z=1.5}
    componentLinearVelScale {z=1}
    gravityOverride {z=-980}
    physicsBodyDefinitions [{
      boundBone {boneName=dyn_BHand_l}
      boxExtents=<25 25 25>
      constraintSetup {angularLimitsMax=<0 7 0>, angularLimitsMin=<0 -5 -5>, angularTarget=<0 0 0>}
      localJointOffset=<0.35 0 0>
     }]
    relativeSpaceBone {boneName=root}
    simSpaceSettings {maxAngularAcceleration=10, maxAngularVelocity=3, simSpaceAngularAlpha=1}
   }
   n14 = Anim Dynamics - Bone: dyn_Bpelvis(ComponentPose=Local To Component(LocalPose=Use cached pose 'Presim'()), Alpha=0.5, ExternalForce="0,0,0") {
    angularDampingOverride=1
    angularSpringConstant=3000
    bLinearSpring=true
    bOverrideLinearDamping=true
    bUseGravityOverride=true
    linearDampingOverride=1
    linearSpringConstant=70
    simulationSpace=RootRelative
    boundBone {boneName=dyn_Bpelvis}
    componentAppliedLinearAccClamp=<500 500 500>
    componentLinearAccScale {z=1.5}
    componentLinearVelScale {z=0.4}
    gravityOverride {z=-980}
    physicsBodyDefinitions [{
      boundBone {boneName=dyn_Bpelvis}
      boxExtents=<30 30 30>
      constraintSetup {angularLimitsMax=<0 0 0>, angularLimitsMin=<0 0 0>, angularTarget=<0 0 0>, linearAxesMax=<0 0 1>, linearAxesMin=<0 0 -1>}
      localJointOffset=<0 0 0>
     }]
    relativeSpaceBone {boneName=root}
    simSpaceSettings {maxAngularAcceleration=30, maxAngularVelocity=5, simSpaceAngularAlpha=0.5}
   }
   n15 = Anim Dynamics - Bone: dyn_BThing_r(ComponentPose=@n14, Alpha=1, ExternalForce="0,0,0") {
    angularDampingOverride=1
    angularSpringConstant=3000
    bAngularSpring=true
    bOverrideAngularDamping=true
    bUseGravityOverride=true
    simulationSpace=BoneRelative
    boundBone {boneName=dyn_BThing_r}
    componentLinearAccScale {z=1.5}
    componentLinearVelScale {z=0.4}
    gravityOverride {z=-980}
    physicsBodyDefinitions [{
      boundBone {boneName=dyn_BThing_r}
      boxExtents=<19 19 19>
      constraintSetup {angularLimitsMax=<0 10 0>, angularLimitsMin=<0 -10 -35>, angularTarget=<0 0 0>}
      localJointOffset=<0.6 0 0>
     }]
    relativeSpaceBone {boneName=root}
    simSpaceSettings {maxAngularAcceleration=30, maxAngularVelocity=5, simSpaceAngularAlpha=1}
   }
   n16 = Anim Dynamics - Bone: dyn_BCalf_r(ComponentPose=@n15, Alpha=1, ExternalForce="0,0,0") {
    angularDampingOverride=1
    angularSpringConstant=3000
    bAngularSpring=true
    bOverrideAngularDamping=true
    bUseGravityOverride=true
    simulationSpace=RootRelative
    boundBone {boneName=dyn_BCalf_r}
    componentLinearAccScale {z=6.1054}
    componentLinearVelScale {z=4.9901}
    gravityOverride {z=-980}
    physicsBodyDefinitions [{
      boundBone {boneName=dyn_BCalf_r}
      boxExtents=<19 19 19>
      constraintSetup {angularLimitsMax=<0 10 15>, angularLimitsMin=<0 -10 -35>, angularTarget=<1 0 0>}
      localJointOffset=<0.6 0 0>
     }]
    relativeSpaceBone {boneName=root}
    simSpaceSettings {maxAngularAcceleration=50, maxAngularVelocity=10, simSpaceAngularAlpha=1}
   }
   n17 = Anim Dynamics - Bone: dyn_BThing_l(ComponentPose=@n16, Alpha=1, ExternalForce="0,0,0") {
    angularDampingOverride=1
    angularSpringConstant=5000
    bAngularSpring=true
    bOverrideAngularDamping=true
    bUseGravityOverride=true
    simulationSpace=RootRelative
    boundBone {boneName=dyn_BThing_l}
    componentLinearAccScale {z=1.5}
    componentLinearVelScale {z=1}
    gravityOverride {z=-980}
    physicsBodyDefinitions [{
      boundBone {boneName=dyn_BThing_l}
      boxExtents=<19 19 19>
      constraintSetup {angularLimitsMax=<0 10 0>, angularLimitsMin=<0 -10 -35>, angularTarget=<0 0 0>}
      localJointOffset=<0.35 0 0>
     }]
    relativeSpaceBone {boneName=root}
    simSpaceSettings {maxAngularAcceleration=30, maxAngularVelocity=6, simSpaceAngularAlpha=1}
   }
   n18 = Anim Dynamics - Bone: dyn_BCalf_l(ComponentPose=@n17, Alpha=1, ExternalForce="0,0,0") {
    angularDampingOverride=1
    angularSpringConstant=5000
    bAngularSpring=true
    bOverrideAngularDamping=true
    bUseGravityOverride=true
    simulationSpace=RootRelative
    boundBone {boneName=dyn_BCalf_l}
    componentLinearAccScale {z=2}
    componentLinearVelScale {z=3.5}
    gravityOverride {z=-980}
    physicsBodyDefinitions [{
      boundBone {boneName=dyn_BCalf_l}
      boxExtents=<19 19 19>
      constraintSetup {angularLimitsMax=<0 10 15>, angularLimitsMin=<0 -10 -35>, angularTarget=<1 0 0>}
      localJointOffset=<0.35 0 0>
     }]
    relativeSpaceBone {boneName=root}
    simSpaceSettings {maxAngularAcceleration=50, maxAngularVelocity=10, simSpaceAngularAlpha=1}
   }
   n19 = Anim Dynamics - Bone: dyn_BSpine1(ComponentPose=@n18, Alpha=1, ExternalForce="0,0,0") {
    angularDampingOverride=1
    angularSpringConstant=2500
    bAngularSpring=true
    bOverrideAngularDamping=true
    bUseGravityOverride=true
    simulationSpace=RootRelative
    boundBone {boneName=dyn_BSpine1}
    componentAppliedLinearAccClamp {z=4000}
    componentLinearAccScale {z=1.5}
    componentLinearVelScale {z=1}
    gravityOverride {z=-980}
    physicsBodyDefinitions [{
      boundBone {boneName=dyn_BSpine1}
      boxExtents=<19 19 19>
      constraintSetup {angularLimitsMax=<0 5 0>, angularLimitsMin=<0 -5 -15>, angularTarget=<1 0 0>}
      localJointOffset=<1.3 0 0>
     }]
    relativeSpaceBone {boneName=root}
    simSpaceSettings {maxAngularAcceleration=30, maxAngularVelocity=7, simSpaceAngularAlpha=1}
   }
   n20 = Anim Dynamics - Bone: dyn_BSpine2(ComponentPose=@n19, Alpha=1, ExternalForce="0,0,0") {
    angularDampingOverride=1
    angularSpringConstant=5000
    bAngularSpring=true
    bOverrideAngularDamping=true
    bUseGravityOverride=true
    simulationSpace=RootRelative
    boundBone {boneName=dyn_BSpine2}
    componentLinearAccScale {z=1.5}
    componentLinearVelScale {z=1}
    gravityOverride {z=-980}
    physicsBodyDefinitions [{
      boundBone {boneName=dyn_BSpine2}
      boxExtents=<19 19 19>
      constraintSetup {angularLimitsMax=<0 2 0>, angularLimitsMin=<0 -2 -15>, angularTarget=<1 0 0>}
      localJointOffset=<1.4 0 0>
     }]
    relativeSpaceBone {boneName=root}
    simSpaceSettings {maxAngularAcceleration=30, maxAngularVelocity=7, simSpaceAngularAlpha=1}
   }
   n21 = Anim Dynamics - Bone: dyn_BHead(ComponentPose=@n20, Alpha=1, ExternalForce="0,0,0") {
    angularDampingOverride=1
    angularSpringConstant=4000
    bAngularSpring=true
    bOverrideAngularDamping=true
    bUseGravityOverride=true
    simulationSpace=RootRelative
    boundBone {boneName=dyn_BHead}
    componentLinearAccScale {z=1}
    componentLinearVelScale {z=0.5}
    gravityOverride {z=-980}
    physicsBodyDefinitions [{
      boundBone {boneName=dyn_BHead}
      boxExtents=<19 19 19>
      constraintSetup {angularLimitsMax=<0 25 1>, angularLimitsMin=<0 -5 -25>, angularTarget=<1 0 0>}
      localJointOffset=<0.8 0 0>
     }]
    relativeSpaceBone {boneName=root}
    simSpaceSettings {maxAngularAcceleration=30, maxAngularVelocity=7, simSpaceAngularAlpha=1}
   }
   n22 = Anim Dynamics - Bone: dyn_BUpperArm_r(ComponentPose=@n21, Alpha=1, ExternalForce="0,0,0") {
    angularDampingOverride=1
    angularSpringConstant=3500
    bAngularSpring=true
    bOverrideAngularDamping=true
    bUseGravityOverride=true
    boundBone {boneName=dyn_BUpperArm_r}
    componentAppliedLinearAccClamp {z=70000}
    componentLinearAccScale {z=4.4055}
    componentLinearVelScale {z=4.7811}
    gravityOverride {z=-980}
    physicsBodyDefinitions [{
      boundBone {boneName=dyn_BUpperArm_r}
      boxExtents=<20 20 20>
      constraintSetup {angularLimitsMax=<0 7.9268 0>, angularLimitsMin=<0 -41.8146 -52.7951>, angularTarget=<1 0 0>}
      localJointOffset=<0.8 0 0>
     }]
    relativeSpaceBone {boneName=root}
    simSpaceSettings {maxAngularAcceleration=50, maxAngularVelocity=10, simSpaceAngularAlpha=1}
   }
   n23 = Anim Dynamics - Bone: dyn_BLoverArm_r(ComponentPose=@n22, Alpha=1, ExternalForce="0,0,0") {
    angularDampingOverride=1
    angularSpringConstant=3500
    bAngularSpring=true
    bOverrideAngularDamping=true
    bUseGravityOverride=true
    simulationSpace=RootRelative
    boundBone {boneName=dyn_BLoverArm_r}
    componentLinearAccScale {z=1.5}
    componentLinearVelScale {z=1}
    gravityOverride {z=-980}
    physicsBodyDefinitions [{
      boundBone {boneName=dyn_BLoverArm_r}
      boxExtents=<20 20 20>
      constraintSetup {angularLimitsMax=<0 5 0>, angularLimitsMin=<0 -12.8 -10>, angularTarget=<0 0 0>}
      localJointOffset=<0.8 0 0>
     }]
    relativeSpaceBone {boneName=root}
    simSpaceSettings {maxAngularAcceleration=50, maxAngularVelocity=10, simSpaceAngularAlpha=1}
   }
   n24 = Anim Dynamics - Bone: dyn_BHand_r(ComponentPose=@n23, Alpha=1, ExternalForce="0,0,0") {
    angularDampingOverride=1
    angularSpringConstant=3500
    bAngularSpring=true
    bOverrideAngularDamping=true
    bUseGravityOverride=true
    simulationSpace=RootRelative
    boundBone {boneName=dyn_BHand_r}
    componentLinearAccScale {z=1.5}
    componentLinearVelScale {z=1}
    gravityOverride {z=-980}
    physicsBodyDefinitions [{
      boundBone {boneName=dyn_BHand_r}
      boxExtents=<20 20 20>
      constraintSetup {angularLimitsMax=<0 5 0>, angularLimitsMin=<0 -20 -5>, angularTarget=<0 0 0>}
      localJointOffset=<0.8 0 0>
     }]
    relativeSpaceBone {boneName=root}
    simSpaceSettings {maxAngularAcceleration=50, maxAngularVelocity=10, simSpaceAngularAlpha=1}
   }
   n25 = Anim Dynamics - Bone: dyn_BUpperArm_l(ComponentPose=@n24, Alpha=1, ExternalForce="0,0,0") {
    angularDampingOverride=1
    angularSpringConstant=5000
    bAngularSpring=true
    bOverrideAngularDamping=true
    bUseGravityOverride=true
    simulationSpace=RootRelative
    boundBone {boneName=dyn_BUpperArm_l}
    componentLinearAccScale {z=2.5}
    componentLinearVelScale {z=1.7}
    gravityOverride {z=-980}
    physicsBodyDefinitions [{
      boundBone {boneName=dyn_BUpperArm_l}
      boxExtents=<20 20 20>
      constraintSetup {angularLimitsMax=<0 25 0>, angularLimitsMin=<0 -5 -27.8>, angularTarget=<0 0 0>}
      localJointOffset=<0.5 0 0>
     }]
    relativeSpaceBone {boneName=root}
    simSpaceSettings {maxAngularAcceleration=50, maxAngularVelocity=10, simSpaceAngularAlpha=1}
   }
   n26 = Anim Dynamics - Bone: dyn_BLoverArm_l(ComponentPose=@n25, Alpha=1, ExternalForce="0,0,0") {
    angularDampingOverride=1
    angularSpringConstant=5000
    bAngularSpring=true
    bOverrideAngularDamping=true
    bUseGravityOverride=true
    simulationSpace=RootRelative
    boundBone {boneName=dyn_BLoverArm_l}
    componentLinearAccScale {z=1.5}
    componentLinearVelScale {z=1}
    gravityOverride {z=-980}
    physicsBodyDefinitions [{
      boundBone {boneName=dyn_BLoverArm_l}
      boxExtents=<20 20 20>
      constraintSetup {angularLimitsMax=<0 15 0>, angularLimitsMin=<0 -5 -10>, angularTarget=<0 0 0>}
      localJointOffset=<0.5 0 0>
     }]
    relativeSpaceBone {boneName=root}
    simSpaceSettings {maxAngularAcceleration=50, maxAngularVelocity=10, simSpaceAngularAlpha=1}
   }
   n27 = Anim Dynamics - Bone: dyn_BHand_l(ComponentPose=@n26, Alpha=1, ExternalForce="0,0,0") {
    angularDampingOverride=1
    angularSpringConstant=5000
    bAngularSpring=true
    bOverrideAngularDamping=true
    bUseGravityOverride=true
    simulationSpace=RootRelative
    boundBone {boneName=dyn_BHand_l}
    componentLinearAccScale {z=1.5}
    componentLinearVelScale {z=1}
    gravityOverride {z=-980}
    physicsBodyDefinitions [{
      boundBone {boneName=dyn_BHand_l}
      boxExtents=<20 20 20>
      constraintSetup {angularLimitsMax=<0 7 0>, angularLimitsMin=<0 -5 -5>, angularTarget=<0 0 0>}
      localJointOffset=<0.5 0 0>
     }]
    relativeSpaceBone {boneName=root}
    simSpaceSettings {maxAngularAcceleration=50, maxAngularVelocity=10, simSpaceAngularAlpha=1}
   }
   Output Pose(Result=Blend Poses by bool(BlendPose_0=Component To Local(ComponentPose=@n27), BlendPose_1=Component To Local(ComponentPose=@n13))) {layerGroup=None}
   n28 = "Transform (Modify) Bone - Bone: pack"(ComponentPose=Local To Component(LocalPose=Input Pose()), Translation=Get backpack_translation(), Rotation=Get backpack_rotate()) {rotationMode=BMM_Additive, translationMode=BMM_Additive, alphaBoolBlend {blendOption=Linear}, boneToModify {boneName=pack}}
   n29 = "Transform (Modify) Bone - Bone: dyn_Bpelvis"(ComponentPose=@n28, Alpha=1, Translation="0.074,-0.576,-0.099", Rotation="0,0,0", Scale="1,1,1") {translationMode=BMM_Additive, translationSpace=BCS_BoneSpace, boneToModify {boneName=dyn_Bpelvis}, translation=<0.074 -0.576 -0.099>}
   n30 = "Transform (Modify) Bone - Bone: dyn_Bpelvis"!(ComponentPose=@n29, bAlphaBoolEnabled:bool=True, Translation="1.5,-1,0", Rotation="0,0,0", Scale="1,1,1") -> (Pose) {
    alphaInputType=Bool
    rotationMode=BMM_Additive
    rotationSpace=BCS_BoneSpace
    translationMode=BMM_Additive
    translationSpace=BCS_BoneSpace
    alphaBoolBlend {blendInTime=0.4, blendOption=Linear, blendOutTime=0.2}
    boneToModify {boneName=dyn_Bpelvis}
    translation=<1.5 -1>
   }
   Save cached pose 'Presim'(Pose=Component To Local(ComponentPose=@n30)) {cacheName=Presim}
  ]}
 EventGraph {nodes [
   Event BlueprintUpdateAnimation()
   n1 = GetOwnerSkeletalMeshComponent()
   n2 = BreakRotator(InRot=BreakTransform(InTransform=GetDeltaTransformFromRefPose(self=@n1, BoneName=spine_05, BaseName=spine_01)).Rotation)
   n3 = MapRangeClamped(Value=@n2.Yaw, InRangeA=-10, InRangeB=-50, OutRangeB=50)
   > Set backpack_rotate(backpack_rotate=MakeRotator(Roll=@n3, Pitch=float / float(A=@n3, B=5)))
   > Set backpack_translation(backpack_translation=MakeVector(Y=MapRangeClamped(Value=@n2.Yaw, InRangeA=-15, InRangeB=-50, OutRangeB=-10)))
   n4 = Get bIsSkyDiving()
   > n5 = Branch(Condition=AND Boolean(A=Get bIsFalling(), B=NOT Boolean(A=@n4)))
   > Set isFallingNotSkydiving(isFallingNotSkydiving:bool=true) -> (Output_Get:bool)
   Set isFallingNotSkydiving(execute=@n5.else, isFallingNotSkydiving=false)
  ]}
]
variables [
 backpack_rotate {type=Rotator}
 backpack_translation {type=Vector}
 isFallingNotSkydiving {type=boolean}
]
